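; Event-triggered gameplay ability for the Clyde shark boss AR that marks hostile player pawns hit by the weapon.
; It confirms the damage source is the avatar's current weapon, then filters the event target to hostile player pawns.
; Valid hits get a team-shared, timed indicator with particles and sound, built from TrackingPistol variables.
; It also applies a marking gameplay effect and reports the mark through a gameplay event and quest stat.
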(event Ability|EventActivateAbilityFromEvent
  (params (EventData :type Gameplay%20Event%20Data%20Structure%20%28by%20ref%29))
  (bind (_eventtag _instigator _target _optionalobject _optionalobject2 _contexthandle _instigatortags _targettags _eventmagnitude _targetdata) (Utilities|Struct|BreakGameplayEventData EventData))
  (bind _returnvalue (Ability|GetAvatarActorfromActorInfo))
  (bind _returnvalue_2 (Fortnite|GetFortPlayerControllerfromActor _returnvalue))
  (bind _returnvalue_3 (Quests|GetQuestManager _returnvalue_2 "Athena"))
  (bind _hitactors (Variables|TrackingPistol|Dynamic|GetHitActors))
  (if (CallFunction|IsDamgefromSourceWeapon EventData)
    (bind _outhitactors (Utilities|HasValidGroupOfHitActors (CallFunction|GetHitActorsfromEventData EventData))
      (:true
        (Variables|TrackingPistol|Dynamic|SetHitActors _outhitactors)
        (bind _returnvalue_4 (Utilities|GetValidAvatarasFortPlayerController true)
          (:Is%20Valid
            (IndicatedManagement|AddActorsToIndicatedList _returnvalue_4 _hitactors (IndicatedManagement|MakeIndicatedActorData :GroupIdentifier (Variables|TrackingPistol|Default|Indicator|GetIndicatorGroupID) :ParticleSystemData (IndicatedManagement|MakeIndicatedActorParticleSystemData (Variables|TrackingPistol|Default|Indicator|GetIndicatorParticleSystem) (Variables|TrackingPistol|Default|Indicator|GetIndicatorActorParamName) "0, 0,100.000000" "0, 0,40.000000" (Variables|TrackingPistol|Default|Indicator|GetIndicatorVectorParamName)) :EventTags _instigatortags :Duration (Variables|TrackingPistol|Default|Indicator|GetIndicatorDuration) :StepTime 0.1 :IndicatorOffset "0.000000,0.000000,90.000000" :IndicatorDBNOOffset "0.000000,0.000000,30.000000" :ShareActorWith "AllTeam" :Sound (Variables|TrackingPistol|Default|Indicator|GetIndicatorSound) :StateImageOverride "None" :StateIndexOverride -1 :bClampToScreen true :bHideStateImage false :bOnlyDisplayWhenStencilingIsDisabled false :bOnlyDisplayIfActorIsVisible false))
            (bind _asfort_player_pawn_athena (Utilities|Casting|CastToFortPlayerPawnAthena (Utilities|Array|Get(acopy) _hitactors))
              (:then
                (Fortnite|ApplyGameplayEffectsToActor _asfort_player_pawn_athena (Variables|TrackingPistol|Default|Indicator|GetEnemyMarkingGE) (Transformation|GetActorLocation _returnvalue_4) false 0 (Ability|GetAbilitySystemComponent _returnvalue_4))
                (Utilities|IsValid _returnvalue
                  (:Is%20Valid
                    (Utilities|IsValid _returnvalue_2
                      (:Is%20Valid
                        (Utilities|IsValid _returnvalue_3
                          (:Is%20Valid
                            (Utilities|FlowControl|Sequence
                              (:then_0
                                ;[1 1 1] Adding gameplay event for QS2
                                (Ability|SendGameplayEventtoActor _returnvalue_2 "(TagName=\"Event.MarkedOpponent\")" (Utilities|Struct|MakeGameplayEventData "(TagName=\"\")" _instigator _target 0 0 _contexthandle _instigatortags _targettags 1.0 _targetdata)))
                              (:then_1
                                (Quest|SendComplexCustomStatEvent _returnvalue_3 0 "(GameplayTags=((TagName=\"Athena.Quests.MarkedEnemy\")))" "(GameplayTags=)" 1))))))))))
              (:CastFailed
                _))
            (Ability|EndAbility))))
      (:false
        _))
    (else
      _))
  (Ability|EndAbility))

(fn GetAvatarAsFortPawn
  (bind _asfort_pawn (Utilities|Casting|CastToFortPawn (Ability|GetAvatarActorfromActorInfo))
    (:then
      (return _asfort_pawn))
    (:CastFailed
      (return 0))))

(fn GetAvatarAsFortPlayerController
  (bind _outfortplayerpawn (CallFunction|GetAvatarasValidFortPlayerPawn)
    (:ValidFortPlayerPawn
      (Utilities|IsValid (Fortnite|GetFortPlayerControllerfromActor _outfortplayerpawn)
        (:Is%20Valid
          (return (Fortnite|GetFortPlayerControllerfromActor _outfortplayerpawn)))
        (:Is%20Not%20Valid
          _)))
    (:AvatarCastFailed
      _))
  (return 0))

(fn GetHitActorsFromEventData
  (params (GameplayEventData :type Gameplay%20Event%20Data%20Structure%20%28by%20ref%29))
  (bind (_eventtag _instigator _target) (Utilities|Struct|BreakGameplayEventData GameplayEventData))
  (bind _output_get (LocalVariables|SetLocalTarget _target))
  (Utilities|IsValid _output_get
    (:Is%20Valid
      (bind _returnvalue (Utilities|GetValidAvatarasFortPlayerController true)
        (:Is%20Valid
          (if (CallFunction|IsPawnandNotFriendly _returnvalue (LocalVariables|GetLocalTarget))
            (return (Utilities|Array|MakeArray (LocalVariables|GetLocalTarget)))
            (else
              (return 0))))))
    (:Is%20Not%20Valid
      (return 0))))

(fn IsDamgeFromSourceWeapon
  (params (GameplayEventData :type Gameplay%20Event%20Data%20Structure%20%28by%20ref%29))
  (bind (_eventtag _instigator _target _optionalobject _optionalobject2 _contexthandle) (Utilities|Struct|BreakGameplayEventData GameplayEventData))
  (bind _asfort_pawn (CallFunction|GetAvatarasFortPawn))
  (Utilities|IsValid _asfort_pawn
    (:Is%20Valid
      (return (Utilities|Equal(Object) (Variables|/Script/FortniteGame.FortPawn|GetCurrentWeapon _asfort_pawn) (Ability|EffectContext|GetSourceObject _contexthandle))))
    (:Is%20Not%20Valid
      (return false))))

(fn IsPawnAndNotFriendly
  (params (OwningController :type Actor%20Object%20Reference) (HitActor :type Actor%20Object%20Reference))
  (bind (_asfort_player_pawn _bsuccess) (Utilities|Casting|CastToFortPlayerPawn HitActor))
  (return (and (Utilities|Enum|Equal(Enum) (Team|GetTeamAffiliation OwningController HitActor) "Hostile") _bsuccess)))
