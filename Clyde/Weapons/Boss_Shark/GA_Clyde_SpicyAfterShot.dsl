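; Gameplay ability for the Clyde Boss_Shark weapon that grants its owner a "spicy" move speed effect after dealing damage.
; On activation it listens in parallel for damage dealt, shield, shielded and overshield gameplay events, each leading to the effect application.
; Only the plain damage event is filtered, requiring a player target and excluding storm damage.
; Applying the effect also starts a five-second timer bound to RemoveGrantedByEffect.
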(event Ability|EventActivateAbility
  (Utilities|FlowControl|Sequence
    (:then_0
      (bind (_asynctaskproxy _payload) (Ability|Tasks|WaitGameplayEvent "(TagName=\"Event.Damage.Dealt\")")
        (:EventReceived
          ; Make sure this isn't triggered by storm damage.
          (if (and (not (GameplayTags|HasTag (.InstigatorTags (Utilities|Struct|BreakGameplayEventData _payload)) "(TagName=\"Gameplay.Damage.OutsideSafeZone\")")) (GameplayTags|HasTag (.TargetTags (Utilities|Struct|BreakGameplayEventData _payload)) "(TagName=\"Pawn.Athena.Player\")"))
            _))))
    (:then_1
      (Ability|Tasks|WaitGameplayEvent "(TagName=\"Event.Damage.Dealt.Shield\")"
        (:EventReceived
          _)))
    (:then_2
      (Ability|Tasks|WaitGameplayEvent "(TagName=\"Event.Damage.Shielded\")"
        (:EventReceived
          _)))
    (:then_3
      (Ability|Tasks|WaitGameplayEvent "(TagName=\"Event.Damage.Dealt.Overshield\")"
        (:EventReceived
          _))))
  (CallFunction|ApplyGameplayEffect))

(event Custom|ApplyGameplayEffect
  (Ability|ApplyGameplayEffectToOwner "/Clyde/Weapons/Boss_Shark/GE_Clyde_SpicyMoveSpeed.GE_Clyde_SpicyMoveSpeed_C")
  (Utilities|Time|SetTimerbyEvent (EventDispatchers|CreateEvent :!event RemoveGrantedByEffect) 5.0))
