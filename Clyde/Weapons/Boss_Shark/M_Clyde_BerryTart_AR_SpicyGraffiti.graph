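# Translucent unlit Niagara sprite material for the BerryTart AR SpicyGraffiti weapon effect, using a drops texture parameter.
# Texture RGB is scaled by the dynamic parameter EmissiveMult for glow, without particle color tint.
# Texture alpha times particle alpha passes through a depth fade, whose distance comes from the dynamic parameter DepthFadeDistance, softening intersections with geometry.
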Material /Clyde/Weapons/Boss_Shark/M_Clyde_BerryTart_AR_SpicyGraffiti
graph {
 properties {
  bAutomaticallySetUsageInEditor=false
  bUsedWithNiagaraSprites=true
  blendMode=BLEND_Translucent
  shadingModel=MSM_Unlit
  usedShadingModels=Unlit
 }
 nodes [
  n1 = TextureSampleParameter2D() {parameterName=Texture, texture=&/T_FX_BerryTart_AR_Drops}
  DepthFade(
   Opacity=Multiply(A=@n1.A, B=ParticleColor().A)
   FadeDistance=DynamicParameter() {paramNames [EmissiveMult, DepthFadeDistance, Param3, Param4]}.DepthFadeDistance
  )
  Multiply(A=@n1.RGB, B=DynamicParameter() {paramNames [EmissiveMult, DepthFadeDistance, Param3, Param4]}.EmissiveMult)
 ]
}
