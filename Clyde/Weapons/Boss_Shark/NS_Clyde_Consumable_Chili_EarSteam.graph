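# Niagara effect for the Clyde chili consumable that blows steam out of a character's ears, attached via sockets on the User.SK_Mesh skeletal mesh.
# Smoke and streak sprite emitters each use two spawn groups, masking velocity to shoot in opposite Z directions for both sides.
# Smoke is offset by System.Offset, streaks shift from pale blue to pink-red, and both scale color by time of day.
# A burst-spawned pink light particle adds glow.
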NiagaraSystem /Clyde/Weapons/Boss_Shark/NS_Clyde_Consumable_Chili_EarSteam
dependencies {
 usedDataInterfaces [
  &NiagaraDataInterfaceCurve
  &NiagaraDataInterfaceVector2DCurve
  &NiagaraDataInterfaceVectorCurve
  &NiagaraDataInterfaceSpriteRendererInfo
 ]
 usedDynamicInputs [
  &ConvertVectorToPosition
  &Add_Vector
  &Multiply_Vector
  &Multiply_VectorByFloat
  &NormalizeVector
 ]
 usedModules [
  &self:SystemSpawnScript.SystemScriptSource.SystemScriptGraph.NiagaraNodeAssignment_2.SetVariables_#1
  &SystemState
  &self:SystemSpawnScript.SystemScriptSource.SystemScriptGraph.NiagaraNodeAssignment_0.SetVariables_#2
  &InterpolateOverTime
  &self:SystemSpawnScript.SystemScriptSource.SystemScriptGraph.NiagaraNodeAssignment_1.SetVariables_#3
 ]
}
emitters {
 Light001 {
  emitterUpdateScript [
   &EmitterState(Life Cycle Mode=::NewEnumerator1, UseLoopDelay=false)
   &SpawnBurst_Instantaneous(Use Spawn Probability=false, UseLoopCountLimit=false)
  ]
  particleSpawnScript [
   &InitializeParticle(
    Color=<rgba 1 0.52 1 1>
    Lifetime=5
    Position Mode=::NewEnumerator0
    Sprite Size=<125 125>
    Sprite Size Mode=::NewEnumerator4
    Write Material Random=true
   )
   &SkeletalMeshLocation(
    MaskSpawnGroup=false
    Mesh Sampling Type=::NewEnumerator5
    Orientation Sampling=::NewEnumerator1
    Skeletal Mesh=$User.SK_Mesh
    Socket Sampling Mode=::NewEnumerator5
   )
   /SetVariables(Particles.LightExponent=5) {isSetParameters=true}
  ]
  particleUpdateScript [&ParticleState()]
  rendererClasses [&NiagaraLightRendererProperties]
  renderers [{rendererClass=&NiagaraLightRendererProperties, rendererIndex=0}]
 }
 Smoke-Left001 {
  emitterUpdateScript [
   &EmitterState()
   &SpawnRate(SpawnGroup=1, SpawnRate=18, Use Spawn Probability=false) {module=SpawnRate001}
   &SpawnRate(SpawnRate=18, Use Spawn Probability=false)
  ]
  particleSpawnScript [
   &InitializeParticle(
    Color Maximum=<rgba 0.65 0.13 0.2167 3>
    Color Minimum=<rgba 1 0.2 0.4667 3>
    Color Mode=::NewEnumerator2
    Lifetime=&RandomRangeFloat(Minimum=0.35, Maximum=0.5)
    Position Mode=::NewEnumerator0
    Sprite Rotation Angle=&Add_Float(
     A=&Multiply_Float(A=8, B=&RandomRangeFloat(Minimum=-1, Maximum=1))
     B=&UniformAOrBFloat(A=0, B=180, Evaluation Type=::NewEnumerator0, Recalculate Random Each Loop=true, Override Seed=false)
    )
    Sprite Rotation Mode=::NewEnumerator0
    Sprite Size=&Vector2DFromFloat(Value=&RandomRangeFloat(Minimum=50, Maximum=60))
    Sprite Size Mode=::NewEnumerator4
    Write Material Random=true
   )
   &SkeletalMeshLocation(
    Interpolate Mesh Sampling=false
    MaskSpawnGroup=false
    Mesh Sampling Type=::NewEnumerator5
    Sampled Position Offset=&self:swizzleXZ(Default:Vector3f=$System.Offset)
    Sampled Position Offset Coordinate Space:ENiagaraSkelMeshTransforms=::NewEnumerator1
    Skeletal Mesh=$User.SK_Mesh
    Socket Sampling Mode=::NewEnumerator5
   )
   &AddVelocity(Velocity=&MaskVectorBySpawnGroup(
     Masked Vector=&RandomRangeVector(Minimum=<-20 -20 200>, Maximum=<20 20 250>)
     Passthrough Vector=&RandomRangeVector(Minimum=<-20 -20 -200>, Maximum=<20 20 -250>)
    ))
  ]
  particleUpdateScript [
   &ParticleState()
   &ScaleSpriteSize(
    Non-Uniform Curve Sprite Scale=DataInterface() {propertyValues {
      XCurve {keys [{tangentMode=RCTM_User, time=0, value=0.3074}, {tangentMode=RCTM_User, time=1, value=1}]}
      YCurve {keys [{tangentMode=RCTM_User, time=0, value=0.0544}, {tangentMode=RCTM_User, time=1, value=1}]}
     }}
    Scale Sprite Size Mode=::NewEnumerator3
   )
   &Drag(Drag=4, Use Linear Drag=true, Use Rotational Drag=true)
   &SolveForcesAndVelocity(Clamp Velocity=false, Limit Acceleration=false)
   /Niagara/Modules/Update/SubUV/SubUVAnimation(Number Of Frames=&FloatFromCurve(FloatCurve=DataInterface() {propertyValues {Curve {keys [{arriveTangent=136.8638, interpMode=RCIM_Cubic, leaveTangent=136.8638, tangentMode=RCTM_User, time=0, value=16}, <at 1 63 RCIM_Cubic>]}}}))
   &ScaleColor(
    Scale Alpha=&FloatFromCurve(
     FloatCurve=DataInterface() {propertyValues {Curve {keys [
         <at 0 0 RCIM_Cubic>
         {arriveTangent=0.1111, interpMode=RCIM_Cubic, leaveTangent=0.1111, time=0.1, value=0.1}
         {arriveTangent=-0.1111, interpMode=RCIM_Cubic, leaveTangent=-0.1111, time=0.9, value=0.1}
         <at 1 0 RCIM_Cubic>
        ]}}}
     Scale Curve=2
    )
    Scale RGB=&VectorFromCurve(VectorCurve=DataInterface() {propertyValues {
       XCurve {keys [<at 0 5 RCIM_Cubic>,<at 0.6 1 RCIM_Cubic>]}
       YCurve {keys [<at 0 5 RCIM_Cubic>,<at 0.6 1 RCIM_Cubic>]}
       ZCurve {keys [<at 0 5 RCIM_Cubic>,<at 0.6 1 RCIM_Cubic>]}
      }})
    ScaleA=true
    ScaleRGB=true
   )
   &DynamicMaterialParameters(
    Index 0 Param 1=&FloatFromCurve(FloatCurve=DataInterface() {propertyValues {Curve {keys [<at 0 1 RCIM_Cubic>,<at 1 -1 RCIM_Cubic>]}}})
    Index 0 Param 2=1
    Index 0 Param 3=1
    Index 0 Param 4=1
    Module.Param0WriteEnabled=true
    Module.Param1WriteEnabled=true
    Module.Param2WriteEnabled=true
    Module.Param3WriteEnabled=true
   )
   &CameraOffset(Camera Offset Amount=3)
   &Color(
    Color=&MultiplyLinearColor(
     Linear Color=&LerpLinearColors(
      StartColor=&RandomRangeLinearColor(Minimum=<rgba 0.5 0.05 0.05 3>, Maximum=<rgba 0.65 0.13 0.2167 3>)
      EndColor=&RandomRangeLinearColor(Minimum=<rgba 1 0.2 0.4667 3>, Maximum=<rgba 0.65 0.13 0.2167 3>)
      LerpFactor=&FloatFromCurve(FloatCurve=DataInterface() {propertyValues {Curve {keys [{tangentMode=RCTM_User, time=0, value=0}, {tangentMode=RCTM_User, time=1, value=1}]}}})
     )
     Scale Factor=&MakeFloat4FromVectorAndFloat(
      "Vector (RGB)"=&VectorFromFloat(Value=&FNTimeOfDayScalar(Night Scale=3, Morning Scale=1, Evening Scale=1))
      "Float (Alpha)"=1
     )
    )
    Scale Alpha=&FloatFromCurve(
     FloatCurve=DataInterface() {propertyValues {Curve {keys [
         <at 0 0 RCIM_Cubic>
         {arriveTangent=0.1111, interpMode=RCIM_Cubic, leaveTangent=0.1111, time=0.1, value=0.1}
         {arriveTangent=-0.1111, interpMode=RCIM_Cubic, leaveTangent=-0.1111, time=0.9, value=0.1}
         <at 1 0 RCIM_Cubic>
        ]}}}
     Scale Curve=2
    )
    Scale Color=&VectorFromCurve(VectorCurve=DataInterface() {propertyValues {
       XCurve {keys [<at 0 5 RCIM_Cubic>,<at 0.6 1 RCIM_Cubic>]}
       YCurve {keys [<at 0 5 RCIM_Cubic>,<at 0.6 1 RCIM_Cubic>]}
       ZCurve {keys [<at 0 5 RCIM_Cubic>,<at 0.6 1 RCIM_Cubic>]}
      }})
   )
  ]
  rendererClasses [&NiagaraSpriteRendererProperties]
  renderers [{rendererClass=&NiagaraSpriteRendererProperties}]
 }
 StreaksLeft {
  emitterUpdateScript [
   &EmitterState()
   &SpawnRate(SpawnGroup=1, SpawnRate=50, Use Spawn Probability=false) {module=SpawnRate001}
   &SpawnRate(SpawnRate=50, Use Spawn Probability=false)
  ]
  particleSpawnScript [
   &InitializeParticle(
    Color Maximum=<rgba 0.2175 0.4105 0.5 1>
    Color Minimum=<rgba 1.74 3.2843 4 0.5>
    Color Mode=::NewEnumerator2
    Lifetime=&RandomRangeFloat(Minimum=0.25, Maximum=0.2)
    Sprite Size=&RandomRangeVector2D(Minimum=<8 30>, Maximum=<6 40>)
    Sprite Size Mode=::NewEnumerator4
    Sprite UV Mode=::NewEnumerator4
    Sprite UV Scale:Vector2f=&MakeCustomVector2DFromBool(Bool=&RandomBool(), True Vector2D=<-1 1>, False Vector2D=<-1 -1>)
    UsePositionOffset=false
    Write Material Random=true
   )
   &AddVelocity(
    Cone Angle=10
    Cone Axis=&MaskVectorBySpawnGroup(Masked Vector=<0 0 -1>, Passthrough Vector=<0 0 1>)
    Distribution Along Cone Axis:NiagaraFloat=&RandomRangeFloat(Minimum=0.02, Maximum=0.05)
    Module.UseConeFalloff=false
    Module.UseVelDistribution=true
    Override Seed=false
    Velocity Mode=::NewEnumerator2
    Velocity Speed=&RandomRangeFloat(Minimum=25, Maximum=50)
   )
   &CameraOffset()
  ]
  particleUpdateScript [
   &ParticleState()
   &ScaleSpriteSize(
    Non-Uniform Curve Sprite Scale=DataInterface() {propertyValues {
      XCurve {keys [
        {arriveTangent=2.3729, interpMode=RCIM_Cubic, leaveTangent=2.3729, leaveTangentWeight=0.3333, tangentMode=RCTM_User, time=0, value=0}
        {
         arriveTangent=1.5
         arriveTangentWeight=0.1667
         interpMode=RCIM_Cubic
         leaveTangent=1.5
         leaveTangentWeight=0.1667
         tangentMode=RCTM_User
         time=0.5
         value=0.8
        }
        {arriveTangent=9.2697, arriveTangentWeight=0.3333, interpMode=RCIM_Cubic, leaveTangent=9.2697, tangentMode=RCTM_User, time=1, value=3}
       ]}
      YCurve {keys [
        {arriveTangent=2.3729, interpMode=RCIM_Cubic, leaveTangent=2.3729, leaveTangentWeight=0.3333, tangentMode=RCTM_User, time=0, value=0}
        {
         arriveTangent=1.5
         arriveTangentWeight=0.1667
         interpMode=RCIM_Cubic
         leaveTangent=1.5
         leaveTangentWeight=0.1667
         tangentMode=RCTM_User
         time=0.5
         value=0.8
        }
        {arriveTangent=0.7753, arriveTangentWeight=0.3333, interpMode=RCIM_Cubic, leaveTangent=0.7753, tangentMode=RCTM_User, time=1, value=1}
       ]}
     }}
    Scale Sprite Size Mode=::NewEnumerator3
   )
   &SubUVAnimation(Number Of Frames=&FloatFromCurve(
     FloatCurve=DataInterface() {propertyValues {Curve {keys [{arriveTangent=136.8638, interpMode=RCIM_Cubic, leaveTangent=136.8638, tangentMode=RCTM_User, time=0, value=16}, <at 1 63 RCIM_Cubic>]}}}
     Scale Curve=0.1
    )) {enabled=false}
   /Niagara/Modules/Update/SubUV/V2/SubUVAnimation(
    Random Start Frame=true
    "Sprite Renderer "=DataInterface() {propertyValues {SpriteRenderer=&self:StreaksLeft_10.NiagaraSpriteRendererProperties_1}}
    SubUV Lookup Index:NiagaraFloat=&FloatFromCurve(
     FloatCurve=DataInterface() {propertyValues {Curve {keys [
         {arriveTangent=136.8638, interpMode=RCIM_Cubic, leaveTangent=136.8638, leaveTangentWeight=0.3333, tangentMode=RCTM_User, time=0, value=16}
         {arriveTangentWeight=0.3333, interpMode=RCIM_Cubic, tangentMode=RCTM_User, time=1, value=63}
        ]}}}
     Scale Curve=&Multiply_Float(A=0.0159, B=0.1)
    )
    UseEndFrame=false
    UseStartFrame=false
   ) {module=SubUVAnimation001}
   &ScaleColor(
    Scale Alpha=&FloatFromCurve(FloatCurve=DataInterface() {propertyValues {Curve {keys [
         <at 0 0 RCIM_Cubic>
         {arriveTangent=0.2857, interpMode=RCIM_Cubic, leaveTangent=0.2857, time=0.15, value=0.1}
         {arriveTangent=-0.2311, interpMode=RCIM_Cubic, leaveTangent=-0.2311, tangentMode=RCTM_User, time=0.35, value=0.1}
         <at 1 0 RCIM_Cubic>
        ]}}})
    Scale RGB=&VectorFromCurve(VectorCurve=DataInterface() {propertyValues {
       XCurve {keys [{arriveTangent=-9.8141, interpMode=RCIM_Cubic, time=0, value=5}, <at 1 1 RCIM_Cubic>]}
       YCurve {keys [<at 0 2.1065 RCIM_Cubic>,<at 1 1 RCIM_Cubic>]}
       ZCurve {keys [<at 0 1.8076 RCIM_Cubic>,<at 1 1 RCIM_Cubic>]}
      }})
    ScaleA=true
    ScaleRGB=true
   )
   &SkeletalMeshLocation(
    MaskSpawnGroup=false
    Mesh Sampling Type=::NewEnumerator5
    Orientation Sampling=::NewEnumerator1
    Sampled Position Offset=<0 1 3>
    Sampled Position Offset Coordinate Space:ENiagaraSkelMeshTransforms=::NewEnumerator1
    Skeletal Mesh=$User.SK_Mesh
    Socket Sampling Mode=::NewEnumerator5
   )
   &SolveForcesAndVelocity(Clamp Velocity=false, Limit Acceleration=false)
   &Color(
    Color=&MultiplyLinearColor(
     Linear Color=&LerpLinearColors(
      StartColor=&RandomRangeLinearColor(Minimum=<rgba 1.74 3.2843 4 0.5>, Maximum=<rgba 0.2175 0.4105 0.5 1>)
      EndColor=&RandomRangeLinearColor(Minimum=<rgba 1 0.2 0.4667 3>, Maximum=<rgba 0.65 0.13 0.2167 3>)
      LerpFactor=&FloatFromCurve(FloatCurve=DataInterface() {propertyValues {Curve {keys [
           {arriveTangent=0.3807, interpMode=RCIM_Cubic, leaveTangent=0.3807, tangentMode=RCTM_User, time=0, value=0}
           {arriveTangent=2.0898, interpMode=RCIM_Cubic, leaveTangent=2.0898, tangentMode=RCTM_User, time=0.7, value=1}
          ]}}})
     )
     Scale Factor=&MakeFloat4FromVectorAndFloat(
      "Vector (RGB)"=&VectorFromFloat(Value=&FNTimeOfDayScalar(Night Scale=3, Morning Scale=1, Evening Scale=1))
      "Float (Alpha)"=1
     )
    )
    Scale Alpha=&FloatFromCurve(FloatCurve=DataInterface() {propertyValues {Curve {keys [
         <at 0 0 RCIM_Cubic>
         {arriveTangent=0.2857, interpMode=RCIM_Cubic, leaveTangent=0.2857, time=0.15, value=0.1}
         {arriveTangent=-0.2311, interpMode=RCIM_Cubic, leaveTangent=-0.2311, tangentMode=RCTM_User, time=0.35, value=0.1}
         <at 1 0 RCIM_Cubic>
        ]}}})
    Scale Color=&VectorFromCurve(VectorCurve=DataInterface() {propertyValues {
       XCurve {keys [{arriveTangent=-9.8141, interpMode=RCIM_Cubic, time=0, value=1}, <at 1 3 RCIM_Cubic>]}
       YCurve {keys [<at 0 1 RCIM_Cubic>,<at 1 3 RCIM_Cubic>]}
       ZCurve {keys [<at 0 1 RCIM_Cubic>,<at 1 3 RCIM_Cubic>]}
      }})
   )
  ]
  rendererClasses [&NiagaraSpriteRendererProperties]
  renderers [{rendererClass=&NiagaraSpriteRendererProperties}]
 }
}
summary {
 systemName=NS_Clyde_Consumable_Chili_EarSteam
 emitters [
  {emitterName=Smoke-Left001, rendererClasses [&NiagaraSpriteRendererProperties]}
  {emitterName=Light001, rendererClasses [&NiagaraLightRendererProperties]}
  {emitterName=StreaksLeft, rendererClasses [&NiagaraSpriteRendererProperties]}
 ]
 userVariables [User.SK_Mesh {
   type=NiagaraDataInterfaceSkeletalMesh
   defaultValue {
    struct=NiagaraExt_VariableValue_DataInterface
    value {dataInterface=&self:NiagaraDataInterfaceSkeletalMesh_0}
   }
  }]
}
